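# Physics asset for the face accessory mesh of the male medium Zombie Elastic cosmetic character (HIS variant), covering five dynamic hair bones.
# Each hair bone gets a small capsule body, and the head gets a larger capsule.
# Every hair body is constrained directly to the head rather than chained, with limited swing and tight twist.
# The first two strands have asymmetric swing limits, while the other three swing equally on both axes.
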PhysicsAsset /BRCosmetics/Characters/Player/Male/Medium/Bodies/M_MED_Zombie_Elastic/Meshes/Parts/HIS/M_MED_Zombie_Elastic_HIS_FaceAcc_Physics
bodies {
 dyn_hair_01 [{shapeType=Capsule, length=3.393, radius=0.907, center=<2 0 0>, rotation=<rot 90 0 0>}]
 dyn_hair_02 [{shapeType=Capsule, length=3.0076, radius=0.505, shapeName=dyn_hair_02_capsule, center=<2.5489 0 0>, rotation=<rot 90 0 0>}]
 dyn_hair_03 [{shapeType=Capsule, length=1.282, radius=0.8786, shapeName=dyn_hair_03_capsule, center=<2 0 0>, rotation=<rot 90 0 0>}]
 dyn_hair_04 [{shapeType=Capsule, length=1.282, radius=0.8786, shapeName=dyn_hair_04_capsule, center=<2 0 0>, rotation=<rot 90 0 0>}]
 dyn_hair_05 [{shapeType=Capsule, length=1.282, radius=0.8786, shapeName=dyn_hair_05_capsule, center=<2 0 0>, rotation=<rot 90 0 0>}]
 head [{shapeType=Capsule, length=1.3419, radius=6.7988, center=<8.649 0.9889 -0.4488>, rotation=<rot 0 -178.6867 -82.5942>}]
}
constraints [
 {bone1Name=dyn_hair_01, bone2Name=head, swing1LimitDegrees=15, swing2LimitDegrees=45, twistLimitDegrees=3}
 {bone1Name=dyn_hair_02, bone2Name=head, swing1LimitDegrees=15, swing2LimitDegrees=45, twistLimitDegrees=3}
 {bone1Name=dyn_hair_03, bone2Name=head, swing1LimitDegrees=35, swing2LimitDegrees=35, twistLimitDegrees=3}
 {bone1Name=dyn_hair_04, bone2Name=head, swing1LimitDegrees=35, swing2LimitDegrees=35, twistLimitDegrees=3}
 {bone1Name=dyn_hair_05, bone2Name=head, swing1LimitDegrees=35, swing2LimitDegrees=35, twistLimitDegrees=3}
]
